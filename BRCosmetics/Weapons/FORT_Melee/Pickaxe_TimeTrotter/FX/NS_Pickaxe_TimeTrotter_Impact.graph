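# Niagara impact effect for the TimeTrotter pickaxe, combining physical-surface debris with flash, flare, glow and galaxy-style elements.
# Debris takes mesh, velocity and gravity settings from the hit surface, with platform-gated collision.
# Flash_Elements packs many sprite, light and mesh renderers into one emitter, each driven by custom attributes and age curves.
# User.Variant selects among six palettes, while time of day, front-end context and quality level adjust brightness, size and spawn counts.
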NiagaraSystem /BRCosmetics/Weapons/FORT_Melee/Pickaxe_TimeTrotter/FX/NS_Pickaxe_TimeTrotter_Impact
dependencies {
 usedDataInterfaces [
  &NiagaraDataInterfaceArrayInt32
  &NiagaraDataInterfaceCurve
  &NiagaraDataInterfaceCamera
  &NiagaraDataInterfaceColorCurve
  &NiagaraDataInterfaceArrayColor
  &NiagaraDataInterfacePlatformSet
  &NiagaraDataInterfaceSpriteRendererInfo
  &NiagaraDataInterfaceVector2DCurve
  &NiagaraDataInterfaceArrayFloat3
 ]
 usedDynamicInputs [
  &SetBoolByIntComparison
  /Game/Effects/Niagara/Modules/Utility/FN_DirectionVectorByPickaxeType
  &FNTimeOfDayColor
 ]
 usedModules [
  &self:SystemSpawnScript.SystemScriptSource.SystemScriptGraph.NiagaraNodeAssignment_2.SetVariables_#1
  &self:SystemSpawnScript.SystemScriptSource.SystemScriptGraph.NiagaraNodeAssignment_3.SetVariables_#2
  &self:SystemSpawnScript.SystemScriptSource.SystemScriptGraph.NiagaraNodeAssignment_4.SetVariables_#3
  &SystemState
  &self:SystemSpawnScript.SystemScriptSource.SystemScriptGraph.NiagaraNodeAssignment_1.SetVariables_#4
  &self:SystemSpawnScript.SystemScriptSource.SystemScriptGraph.NiagaraNodeAssignment_0.SetVariables_#5
  &self:SystemSpawnScript.SystemScriptSource.SystemScriptGraph.NiagaraNodeAssignment_5.SetVariables_#6
 ]
}
emitters {
 Directional_FN__PhysMatDebris001 {
  emitterUpdateScript [
   &EmitterState()
   &SpawnBurst_Instantaneous(
    Spawn Count=&RandomRangeInt(Minimum=5, Maximum=10)
    Use Spawn Probability=false
    UseLoopCountLimit=false
   )
  ]
  particleSpawnScript [
   /SetVariables(Particles.BiasedRandom=&FloatFromCurve(
     FloatCurve=DataInterface() {propertyValues {Curve {keys [
         <at 0.2 1 RCIM_Cubic>
         {arriveTangent=-3.3394, interpMode=RCIM_Cubic, leaveTangent=-3.3394, tangentMode=RCTM_User, time=0.5, value=0.5}
         <at 0.8 0 RCIM_Cubic>
        ]}}}
     CurveIndex=&RandomRangeFloat(Minimum=0, Maximum=1)
    )) {isSetParameters=true}
   &InitializeParticle(
    Color Mode=::NewEnumerator0
    Lifetime=&RandomRangeFloat(Minimum=0.5, Maximum=1.25)
    Mass=&Lerp_Float(A=0.6, B=1.2, Alpha=$Particles.BiasedRandom)
    Mass Mode=::NewEnumerator1
    Mesh Scale=&Multiply_VectorByFloat(
     Vector=&RandomRangeVector(Minimum=<0.9 0.4 0.5>, Maximum=<1 1 1>)
     Float=&Lerp_Float(A=0.7, B=1, Alpha=$Particles.BiasedRandom)
    )
    Mesh Scale Mode=::NewEnumerator4
    UsePositionOffset=false
    Write Material Random=true
   )
   &FN_PhysicalSurface(
    Dirt Mesh Count=2
    Glass Mesh Count=3
    Ice Mesh Count=2
    Metal Gravity=&RandomRangeFloat(Minimum=-350, Maximum=-1800)
    Metal Mesh Count=2
    Metal Velocity=&RandomRangeFloat(Minimum=80, Maximum=375)
    Metal Velocity Cone Angle=130
    Physical Surface=$User.PhysicalSurface
    Stone Gravity=&RandomRangeFloat(Minimum=-700, Maximum=-1250)
    Stone Mesh Count=2
    Stone Velocity=&RandomRangeFloat(Minimum=80, Maximum=550)
    Stone Velocity Cone Angle=120
    Wood Gravity=&RandomRangeFloat(Minimum=-450, Maximum=-900)
    Wood Mesh Count=2
    Wood Scale=1.5
    Wood Velocity=&RandomRangeFloat(Minimum=80, Maximum=600)
    Wood Velocity Cone Angle=130
   )
   &AddVelocityInCone(
    Cone Angle=$Output.FN_PhysicalSurface.SurfaceResponseConeAngle
    Cone Axis=$System.DirectionVector
    Module.Use Velocity Falloff On Cone Axis=false
    Velocity Strength=$Output.FN_PhysicalSurface.SurfaceResponseVelocity
   )
   &InitialMeshOrientation(Use Rotation Vector=false)
   &AddRotationalVelocity(Rotation Rate=&RandomVector(Vector Scale=&RandomRangeFloat(Minimum=1, Maximum=2)))
   &ApplyInitialForces()
   /SetVariables(Particles.GravityOverrideValue=$Output.FN_PhysicalSurface.GravityOverride) {isSetParameters=true}
  ]
  particleUpdateScript [
   &ParticleState()
   &GravityForce(Gravity=$Particles.GravityOverrideValue)
   &Collision(
    Advanced Aging Rate=1
    Age Colliding Particles=false
    CPU Collision Trace Channel=ECC_WorldStatic
    Collision Enabled=&Boolean_AndOperation(
     Bool1=&IsPlatformSetActive(New Platform Set=DataInterface() {propertyValues {Platforms {qualityLevelMask=-8}}})
     Bool2=&Boolean_AndOperation(
      Bool1=&SetBoolByFloatComparison(A=$Particles.Initial.Lifetime, B=1)
      Bool2=&SetBoolByFloatComparison(A=$Particles.Age, B=0.12)
     )
    )
    Control Roll On Collision=true
    Friction=1
    Friction During a Bounce=0.25
    Max CPU Trace Length=100
    Module.Control Rest State Via Surface Normal=false
    Module.EnableMaxCollisionCount=false
    Module.Randomize Collision Normal Vector=true
    Randomize Collision Normal=0.25
    ScaleRateOfRotation=false
   )
   &SolveForcesAndVelocity(Clamp Velocity=false, Limit Acceleration=false)
   &ScaleMeshSize(Scale Factor=&VectorFromFloat(Value=&FloatFromCurve(FloatCurve=DataInterface() {propertyValues {Curve {keys [<at 0.6 1 RCIM_Cubic>, {arriveTangent=-6.308, interpMode=RCIM_Cubic, leaveTangent=-6.308, tangentMode=RCTM_User, time=1, value=0}]}}})))
   &DynamicMaterialParameters(Module.Param0WriteEnabled=true, Module.Param1WriteEnabled=true, Module.Param2WriteEnabled=false, Module.Param3WriteEnabled=false)
  ]
  rendererClasses [&NiagaraSpriteRendererProperties, &NiagaraMeshRendererProperties]
  renderers [
   {rendererClass=&NiagaraSpriteRendererProperties}
   {rendererClass=&NiagaraMeshRendererProperties, rendererIndex=1}
   {rendererClass=&NiagaraMeshRendererProperties, rendererIndex=2}
   {rendererClass=&NiagaraMeshRendererProperties, rendererIndex=3}
   {rendererClass=&NiagaraMeshRendererProperties, rendererIndex=4}
   {rendererClass=&NiagaraMeshRendererProperties, rendererIndex=5}
   {rendererClass=&NiagaraMeshRendererProperties, rendererIndex=6}
   {rendererClass=&NiagaraMeshRendererProperties, rendererIndex=7}
   {rendererClass=&NiagaraMeshRendererProperties, rendererIndex=8}
  ]
 }
 Flash_Elements {
  emitterUpdateScript [
   &EmitterState()
   &SpawnBurst_Instantaneous(Spawn Group=1, Use Spawn Probability=false, UseLoopCountLimit=false) {module=SpawnBurst_Instantaneous001}
   &SpawnBurst_Instantaneous(
    Spawn Count=&RandomRangeInt(
     Minimum=&SelectIntFromArray(Direct Array Index=$Engine.QualityLevel, Int Selection Array=DataInterface() {propertyValues {IntData [5,6,8,10,10]}})
     Maximum=&SelectIntFromArray(Direct Array Index=$Engine.QualityLevel, Int Selection Array=DataInterface() {propertyValues {IntData [8,9,10,15,15]}})
    )
    Use Spawn Probability=false
    UseLoopCountLimit=false
   )
   &SpawnBurst_Instantaneous(
    Spawn Count=&RandomRangeInt(
     Minimum=&SelectIntFromArray(Direct Array Index=$Engine.QualityLevel, Int Selection Array=DataInterface() {propertyValues {IntData [1,2,2,3,3]}})
     Maximum=&SelectIntFromArray(Direct Array Index=$Engine.QualityLevel, Int Selection Array=DataInterface() {propertyValues {IntData [2,3,4,5,5]}})
    )
    Spawn Group=2
    Use Spawn Probability=false
    UseLoopCountLimit=false
   ) {module=SpawnBurst_Instantaneous002}
  ]
  particleSpawnScript [
   &InitializeParticle(
    Color Mode=::NewEnumerator0
    Lifetime=&MakeCustomFloatFromBool(Bool=$System.IsGalaxy, True Float=0.55, False Float=0.35)
    Position Offset=$System.FEImpactOffset
    Position Offset Coordinate Space=::Local
    UsePositionOffset=true
    Write Material Random=true
   )
   &ShapeLocation(
    MaskSpawnGroup=false
    Override Seed=false
    Sphere Radius=&MakeCustomFloatFromBool(Bool=$System.IsGalaxy, True Float=1, False Float=1)
    UseOwnerScale=true
   )
   &SpriteFacingAndAlignment(Sprite Alignment=$Particles.ShapeLocation.ShapeNormal, WriteAlignment=true, WriteFacing=false)
   &FN_PerParticleLOD(Enable SpriteSize Distance Scaling=true, Sprite Size Scale Far=<2.5 2.5>)
   /SetVariables(Particles.SpriteSize=<1 1>, Particles.VisibilityTag=&SpawnGroup()) {isSetParameters=true}
   /SetVariables(
    Particles.Flash Spike Color:LinearColor=&MergeHelper_ColorSwitch(
     0=<rgba 0.6203 0.0547 0.3184 0.5>
     1=<rgba 0.0833 0 1 0.5>
     2=<rgba 0.1 0.056 0.02 1>
     3=<rgba 0.025 0.1 0.25 0.8>
     4=<rgba 0.1306 0.2119 0.55 3>
     5=<rgba 0.1425 0.2311 0.6 3>
     Index=&MakeIntFromFloat(Float=$User.Variant)
     Num=::NewEnumerator4
    )
    Particles.Flash Spike Lifetime:NiagaraFloat=&RandomRangeFloat(Minimum=0.05, Maximum=0.08)
    Particles.Flash Spike Sprite Size:Vector2f=&FN_FrontEndVector2Modifier(
     Front End:Vector2f=&RandomRangeVector2D(Minimum=<2 25>, Maximum=<7 40>)
     Gameplay:Vector2f=&RandomRangeVector2D(Minimum=<3 55>, Maximum=<10 70>)
    )
   ) {isSetParameters=true}
   /SetVariables(
    Particles.Flash Star Cam Offset=15
    Particles.Flash Star Lifetime=0.15
    Particles.Flash Star Rotation:NiagaraFloat=&RandomRangeFloat(Minimum=0, Maximum=360)
    Particles.Flash Star Sprite Size:Vector2f=&Multiply_Vector2D_ByFloat(
     Float=&Multiply_Float(
      A=&MakeCustomFloatFromBool(Bool=$User.bIsFrontEnd, True Float=0.5, False Float=1)
      B=&RandomRangeFloat(Minimum=120, Maximum=150)
     )
     Vector2D=$Particles.SpriteSize
    )
    Particles.Flash Star Sprite Size Alt:Vector2f=&Multiply_Vector2D_ByFloat(
     Float=&Multiply_Float(
      A=0.85
      B=&Multiply_Float(
       A=&MakeCustomFloatFromBool(Bool=$User.bIsFrontEnd, True Float=0.5, False Float=1)
       B=&RandomRangeFloat(Minimum=120, Maximum=150)
      )
     )
     Vector2D=$Particles.SpriteSize
    )
    Particles.Flash Star SubUV Index:NiagaraFloat=&RandomRangeFloat(Minimum=0, Maximum=3)
    Particles.Flash Star SubUV Index Alt:NiagaraFloat=&RandomRangeFloat(Minimum=0, Maximum=3)
   ) {isSetParameters=true}
   /SetVariables(
    Particles.Flash Flare Color:LinearColor=<rgba 1 1 1 0.7>
    Particles.Flash Flare Lifetime:NiagaraFloat=&FN_FrontEndFloatModifier(Gameplay=0.2, Front End=0.125)
    Particles.Flash Flare Sprite Size:Vector2f=&Vector2DFromFloat(Value=&Multiply_Float(A=$System.FrontendScale, B=160))
   ) {isSetParameters=true}
   /SetVariables(
    Particles.GalaxySmoke Lifetime:NiagaraFloat=&RandomRangeFloat(Minimum=0.25, Maximum=0.35)
    Particles.GalaxyWave Lifetime=0.5
   ) {isSetParameters=true}
   /SetVariables(Particles.Dark_Glow_CameraOffset=&MakeCustomFloatFromBool(Bool=$User.bIsFrontEndPreview, True Float=1, False Float=50)) {isSetParameters=true}
   &InitialMeshOrientation(
    Mesh Orientation Mode=::NewEnumerator7
    Rotation=&FN_FrontEndVector3Modifier(Gameplay=<0 0.25 0>, Front End=<0 0.25 0>)
    Use Rotation Vector=true
   )
   &AddVelocity(Velocity=&FN_FrontEndVector3Modifier(
     Gameplay=<10 0 0>
     Front End=&Multiply_Vector(A=&ReturnCameraVector(Camera Vectors=::NewEnumerator1, Coordinate Space=::Local), B=<-1 -1 -1>)
    ))
   &ApplyInitialForces()
  ]
  particleUpdateScript [
   &ParticleState()
   /SetVariables(
    Particles.Update GalaxySmoke Lifetime:NiagaraFloat=&Divide_Float(A=$Particles.Age, B=$Particles.Initial.GalaxySmoke Lifetime)
    Particles.Update GalaxyWave Lifetime:NiagaraFloat=&Divide_Float(A=$Particles.Age, B=$Particles.Initial.GalaxyWave Lifetime)
   ) {isSetParameters=true}
   /SetVariables(
    Particles.Update Flash Flare Age:NiagaraFloat=&Divide_Float(A=$Particles.Age, B=$Particles.Initial.Flash Flare Lifetime)
    Particles.Update Flash Spike Age:NiagaraFloat=&Divide_Float(A=$Particles.Age, B=$Particles.Initial.Flash Spike Lifetime)
    Particles.Update Flash Star Age:NiagaraFloat=&Divide_Float(A=$Particles.Age, B=$Particles.Initial.Flash Star Lifetime)
   ) {isSetParameters=true}
   /SetVariables(
    Particles.Update Flash Spike Color:LinearColor=&MultiplyLinearColor(
     Linear Color=&MakeLinearColorFromVectorAndFloat(
      "Vector (RGB)"=&MakeVectorFromLinearColorRGB(Color=$Particles.Initial.Flash Spike Color)
      "Float (Alpha)"=&FloatFromCurve(
       FloatCurve=DataInterface() {propertyValues {Curve {keys [<at 0.7 1>,<at 1 0>]}}}
       CurveIndex=$Particles.Update Flash Spike Age
       Scale Curve=&FNTimeOfDayScalar(Night Scale=0.835, Morning Scale=0.85, Evening Scale=0.875)
      )
     )
     Scale Factor=&MakeVector4_FromLinearColor(LinearColor=$System.TODColorMultiplier)
    )
    Particles.Update Flash Spike Sprite Size:Vector2f=&Multiply_Vector2D_ByFloat(
     Float=&FloatFromCurve(FloatCurve=DataInterface() {propertyValues {Curve {keys [<at 0 0.6>,<at 1 2>]}}}, CurveIndex=$Particles.Update Flash Spike Age)
     Vector2D=$Particles.Initial.Flash Spike Sprite Size
    )
    Particles.Update Flash Star Color:LinearColor=&MultiplyLinearColor(
     Linear Color=&MergeHelper_ColorSwitch(
      0=&ColorFromCurve(
       ColorCurve=DataInterface() {propertyValues {
         AlphaCurve {keys [<at 0 1>,<at 0.8992 0>]}
         BlueCurve {keys [<at 0 1.8436>,<at 0.6127 0.3613>]}
         GreenCurve {keys [<at 0 0.3616>,<at 0.6127 0.0862>]}
         RedCurve {keys [<at 0 3>,<at 0.6127 0.5>]}
        }}
       CurveIndex=$Particles.Update Flash Star Age
      )
      1=&ColorFromCurve(
       ColorCurve=DataInterface() {propertyValues {
         AlphaCurve {keys [<at 0 1>,<at 0.8992 0>]}
         BlueCurve {keys [<at 0 2>,<at 0.6127 0.5>]}
         GreenCurve {keys [<at 0 0.3021>,<at 0.6127 0>]}
         RedCurve {keys [<at 0 0.4425>,<at 0.6127 0.0833>]}
        }}
       CurveIndex=$Particles.Update Flash Star Age
      )
      2=&ColorFromCurve(
       ColorCurve=DataInterface() {propertyValues {
         AlphaCurve {keys [<at 0 2.5>,<at 0.8992 0>]}
         BlueCurve {keys [<at 0 0.2292>,<at 1 0.0128>]}
         GreenCurve {keys [<at 0 1.1146>,<at 1 0.0315>]}
         RedCurve {keys [<at 0 2>,<at 1 0.0638>]}
        }}
       CurveIndex=$Particles.Update Flash Star Age
      )
      3=&ColorFromCurve(
       ColorCurve=DataInterface() {propertyValues {
         AlphaCurve {keys [<at 0 2>,<at 0.8992 0>]}
         BlueCurve {keys [<at 0 1>,<at 1 0.0464>]}
         GreenCurve {keys [<at 0 0.4458>,<at 1 0.0233>]}
         RedCurve {keys [<at 0 0.05>,<at 1 0.0068>]}
        }}
       CurveIndex=$Particles.Update Flash Star Age
      )
      4=&ColorFromCurve(
       ColorCurve=DataInterface() {propertyValues {
         AlphaCurve {keys [<at 0 2>,<at 0.8992 0>]}
         BlueCurve {keys [<at 0.0023 0.2292>,<at 0.2231 0.2292>,<at 0.6127 0.2292>]}
         GreenCurve {keys [<at 0.0023 0.0458>,<at 0.2231 0.0757>,<at 0.6127 0.0917>]}
         RedCurve {keys [<at 0.0023 0.1986>,<at 0.2231 0.2164>,<at 0.6127 0.2292>]}
        }}
       CurveIndex=$Particles.Update Flash Star Age
      )
      5=&ColorFromCurve(
       ColorCurve=DataInterface() {propertyValues {
         AlphaCurve {keys [<at 0 2>,<at 0.8992 0>]}
         BlueCurve {keys [<at 0.0023 3>,<at 0.6127 0.5>]}
         GreenCurve {keys [<at 0.0023 0.0309>,<at 0.6127 0.0417>]}
         RedCurve {keys [<at 0.0023 0.5258>,<at 0.6127 0>]}
        }}
       CurveIndex=$Particles.Update Flash Star Age
      )
      Index=&MakeIntFromFloat(Float=$User.Variant)
      Num=::NewEnumerator4
     )
     Scale Factor=&MakeVector4FromVectorAndFloat(
      "Vector (RGB)"=&MakeVectorFromLinearColorRGB(Color=$System.TODColorMultiplier)
      "Float (Alpha)"=&Multiply_Float(
       A=&MakeFloatFromLinearColor(LinearColor=$System.TODColorMultiplier, Channel:ENiagaraFloatFromLinearColorOptions=::NewEnumerator3)
       B=&FNTimeOfDayScalar(Night Scale=0.835, Morning Scale=0.85, Evening Scale=0.875)
      )
     )
    )
    Particles.Update Flash Star Color Alt:LinearColor=&MultiplyLinearColor(
     Linear Color=&MergeHelper_ColorSwitch(
      0=&ColorFromCurve(
       ColorCurve=DataInterface() {propertyValues {
         AlphaCurve {keys [<at 0 2>,<at 1 0>]}
         BlueCurve {keys [<at 0.1978 0.0989>]}
         GreenCurve {keys [<at 0.1978 0.0196>]}
         RedCurve {keys [<at 0.1978 0.15>]}
        }}
       CurveIndex=$Particles.Update Flash Star Age
      )
      1=&ColorFromCurve(
       ColorCurve=DataInterface() {propertyValues {
         AlphaCurve {keys [<at 0 2>,<at 1 0>]}
         BlueCurve {keys [<at 0.1978 0.15>]}
         GreenCurve {keys [<at 0.1978 0.023>]}
         RedCurve {keys [<at 0.1978 0.0336>]}
        }}
       CurveIndex=$Particles.Update Flash Star Age
      )
      2=&ColorFromCurve(
       ColorCurve=DataInterface() {propertyValues {
         AlphaCurve {keys [<at 0 3>,<at 0.8992 0>]}
         BlueCurve {keys [<at 0.1561 0.0413>,<at 0.8536 0.0278>]}
         GreenCurve {keys [<at 0.1561 0.3514>,<at 0.8536 0.0921>]}
         RedCurve {keys [<at 0.1561 0.6615>,<at 0.8536 0.2031>]}
        }}
       CurveIndex=$Particles.Update Flash Star Age
      )
      3=&ColorFromCurve(
       ColorCurve=DataInterface() {propertyValues {
         AlphaCurve {keys [<at 0 2>,<at 0.8992 0>]}
         BlueCurve {keys [<at 0.1978 1>,<at 0.4068 0.6551>,<at 0.6127 0.0538>]}
         GreenCurve {keys [<at 0.1978 0.8562>,<at 0.4068 0.5965>,<at 0.6127 0.0494>]}
         RedCurve {keys [<at 0.1978 0.1374>,<at 0.4068 0.1088>,<at 0.6127 0.0011>]}
        }}
       CurveIndex=$Particles.Update Flash Star Age
      )
      4=&ColorFromCurve(
       ColorCurve=DataInterface() {propertyValues {
         AlphaCurve {keys [<at 0 2>,<at 0.8992 0>]}
         BlueCurve {keys [<at 0.1978 1>,<at 0.4068 0.8958>,<at 0.6127 0.4219>]}
         GreenCurve {keys [<at 0.1978 0.6043>,<at 0.4068 0.6852>,<at 0.6127 0.2153>]}
         RedCurve {keys [<at 0.1978 0.2087>,<at 0.4068 0.0532>,<at 0.6127 0.0087>]}
        }}
       CurveIndex=$Particles.Update Flash Star Age
      )
      5=&ColorFromCurve(
       ColorCurve=DataInterface() {propertyValues {
         AlphaCurve {keys [<at 0 2>,<at 0.8992 0>]}
         BlueCurve {keys [<at 0.1978 0.3652>,<at 0.4068 0.6087>,<at 0.6127 0.2696>]}
         GreenCurve {keys [<at 0.1978 0.2542>,<at 0.4068 0.4>,<at 0.6127 0.1876>]}
         RedCurve {keys [<at 0.1978 0.0647>,<at 0.4068 0.1078>,<at 0.6127 0.0478>]}
        }}
       CurveIndex=$Particles.Update Flash Star Age
      )
      Index=&MakeIntFromFloat(Float=$User.Variant)
      Num=::NewEnumerator4
     )
     Scale Factor=&MakeVector4_FromLinearColor(LinearColor=$System.TODColorMultiplier)
    )
    Particles.Update Flash Star Sprize Size:Vector2f=&Multiply_Vector2D_ByFloat(
     Float=&FloatFromCurve(
      FloatCurve=DataInterface() {propertyValues {Curve {keys [
          {arriveTangent=2.7013, interpMode=RCIM_Cubic, leaveTangent=2.7013, tangentMode=RCTM_User, time=0, value=0}
          {interpMode=RCIM_Cubic, tangentMode=RCTM_User, time=1, value=1}
         ]}}}
      CurveIndex=$Particles.Update Flash Star Age
     )
     Vector2D=$Particles.Initial.Flash Star Sprite Size
    )
    Particles.Update Flash Star Sprize Size Alt:Vector2f=&Multiply_Vector2D_ByFloat(
     Float=&FloatFromCurve(
      FloatCurve=DataInterface() {propertyValues {Curve {keys [
          {arriveTangent=2.7013, interpMode=RCIM_Cubic, leaveTangent=2.7013, tangentMode=RCTM_User, time=0, value=0}
          {interpMode=RCIM_Cubic, tangentMode=RCTM_User, time=1, value=1}
         ]}}}
      CurveIndex=$Particles.Update Flash Star Age
     )
     Vector2D=$Particles.Initial.Flash Star Sprite Size Alt
    )
   ) {isSetParameters=true}
   /SetVariables(Particles.Update Flash Flare Color:LinearColor=&MultiplyLinearColor(
     Linear Color=&MakeLinearColorFromVectorAndFloat(
      "Vector (RGB)"=&Multiply_Vector(
       A=&MakeVectorFromLinearColorRGB(Color=$Particles.Flash Flare Color)
       B=&MakeVectorFromLinearColorRGB(Color=&MergeHelper_ColorSwitch(
         0=&ColorFromCurve(ColorCurve=DataInterface() {propertyValues {
            AlphaCurve {keys [<at 0 1 RCIM_Cubic>,<at 1 0 RCIM_Cubic>]}
            BlueCurve {keys [<at 0 1.2503>,<at 0.6579 0.2>]}
            GreenCurve {keys [<at 0 0.502>,<at 0.6579 0.0338>]}
            RedCurve {keys [<at 0 2>,<at 0.6579 0.1959>]}
           }})
         1=&ColorFromCurve(ColorCurve=DataInterface() {propertyValues {
            AlphaCurve {keys [<at 0 1 RCIM_Cubic>,<at 1 0 RCIM_Cubic>]}
            BlueCurve {keys [<at 0 2>,<at 0.6579 0.2>]}
            GreenCurve {keys [<at 0 1.1479>,<at 0.6579 0.0667>]}
            RedCurve {keys [<at 0 0.2959>,<at 0.6579 0>]}
           }})
         2=&ColorFromCurve(ColorCurve=DataInterface() {propertyValues {
            AlphaCurve {keys [<at 0 1 RCIM_Cubic>,<at 1 1 RCIM_Cubic>]}
            BlueCurve {keys [<at 0 0.1942 RCIM_Cubic>,<at 0.713 0 RCIM_Cubic>]}
            GreenCurve {keys [<at 0 0.6643 RCIM_Cubic>,<at 0.713 0.3833 RCIM_Cubic>]}
            RedCurve {keys [<at 0 1 RCIM_Cubic>,<at 0.713 1 RCIM_Cubic>]}
           }})
         3=&ColorFromCurve(ColorCurve=DataInterface() {propertyValues {
            AlphaCurve {keys [<at 0 1 RCIM_Cubic>,<at 1 1 RCIM_Cubic>]}
            BlueCurve {keys [<at 0 1 RCIM_Cubic>,<at 0.713 0.8 RCIM_Cubic>]}
            GreenCurve {keys [<at 0 0.5833 RCIM_Cubic>,<at 0.713 0.3333 RCIM_Cubic>]}
            RedCurve {keys [<at 0 0 RCIM_Cubic>,<at 0.713 0 RCIM_Cubic>]}
           }})
         4=&ColorFromCurve(ColorCurve=DataInterface() {propertyValues {
            AlphaCurve {keys [<at 0 1 RCIM_Cubic>,<at 1 1 RCIM_Cubic>]}
            BlueCurve {keys [<at 0 1 RCIM_Cubic>,<at 0.713 0.8 RCIM_Cubic>]}
            GreenCurve {keys [<at 0 0.4667 RCIM_Cubic>,<at 0.713 0.2017 RCIM_Cubic>]}
            RedCurve {keys [<at 0 0 RCIM_Cubic>,<at 0.713 0.7003 RCIM_Cubic>]}
           }})
         5=&ColorFromCurve(ColorCurve=DataInterface() {propertyValues {
            AlphaCurve {keys [<at 0 1 RCIM_Cubic>,<at 1 1 RCIM_Cubic>]}
            BlueCurve {keys [<at 0 2 RCIM_Cubic>,<at 0.713 0.2 RCIM_Cubic>]}
            GreenCurve {keys [<at 0 0.6667 RCIM_Cubic>,<at 0.713 0.0629 RCIM_Cubic>]}
            RedCurve {keys [<at 0 0 RCIM_Cubic>,<at 0.713 0.0504 RCIM_Cubic>]}
           }})
         Index=&MakeIntFromFloat(Float=$User.Variant)
         Num=::NewEnumerator4
        ))
      )
      "Float (Alpha)"=&FloatFromCurve(
       FloatCurve=DataInterface() {propertyValues {Curve {keys [<at 0 1>,<at 1 0>]}}}
       CurveIndex=$Particles.Update Flash Flare Age
       Scale Curve=&FNTimeOfDayScalar(Night Scale=0.835, Morning Scale=0.85, Evening Scale=0.875)
      )
     )
     Scale Factor=&MakeVector4_FromLinearColor(LinearColor=$System.TODColorMultiplier)
    )) {isSetParameters=true}
   &FN_Light_Attributes(
    Color=&MergeHelper_ColorSwitch(
     0=<rgba 3 0.3403 2.0645 1>
     1=<rgba 0.489 1.326 3 1>
     2=<rgba 3 1.875 0.75 1>
     3=<rgba 0.15 1.3375 3 1>
     4=<rgba 0.4174 1.4935 3 1>
     5=<rgba 0.4174 0.8478 3 1>
     Index=&MakeIntFromFloat(Float=$User.Variant)
     Num=::NewEnumerator4
    )
    Create Color Attribute=true
    Create Exponent Attribute=true
    Create Position Attribute=true
    Create Radius Attribute=true
    Exponent=5
    Position=<15 0 0>
    Radius=90
    Scale Color=&MakeCustomFloatFromBool(
     Bool=$System.IsGalaxy
     True Float=&FloatFromCurve(
      FloatCurve=DataInterface() {propertyValues {Curve {keys [
          {interpMode=RCIM_Cubic, tangentMode=RCTM_User, time=0, value=1}
          {arriveTangent=-2.4145, interpMode=RCIM_Cubic, leaveTangent=-2.4145, tangentMode=RCTM_User, time=0.63, value=0}
         ]}}}
      Scale Curve=10
     )
     False Float=&FloatFromCurve(
      FloatCurve=DataInterface() {propertyValues {Curve {keys [<at 0 1>, {arriveTangent=-2.0035, leaveTangent=-2.0035, time=0.8, value=0}]}}}
      Scale Curve=10
     )
    )
   )
   &FN_Sprite_Attributes(
    Color=&MultiplyLinearColor(
     Linear Color=&ColorFromCurve(ColorCurve=DataInterface() {propertyValues {
        AlphaCurve {keys [<at 0 2.5 RCIM_Cubic>,<at 1 0 RCIM_Cubic>]}
        BlueCurve {keys [<at 0 0.3107>, {arriveTangent=-0.3111, interpMode=RCIM_Cubic, leaveTangent=-0.3111, time=0.1517, value=0}, <at 0.9989 0>]}
        GreenCurve {keys [<at 0 1.7523>, {arriveTangent=-1.7543, interpMode=RCIM_Cubic, leaveTangent=-1.7543, time=0.1517, value=0}, <at 0.9989 0>]}
        RedCurve {keys [<at 0 5>, {arriveTangent=-5.0055, interpMode=RCIM_Cubic, leaveTangent=-5.0055, time=0.1517, value=0}, <at 0.9989 0>]}
       }})
     Scale Factor=&MakeVector4FromVectorAndFloat("Vector (RGB)"=&MakeVectorFromLinearColorRGB(Color=$System.TODColorMultiplier), "Float (Alpha)"=1)
    )
    Create Color Attribute=true
    Create SpriteSize Attribute=true
    SpriteSize=&Vector2DFromFloat(Value=&FloatFromCurve(
      FloatCurve=DataInterface() {propertyValues {Curve {keys [
          {arriveTangent=7.1084, interpMode=RCIM_Cubic, leaveTangent=7.1084, tangentMode=RCTM_User, time=0, value=0.375}
          {arriveTangent=-2.821, interpMode=RCIM_Cubic, leaveTangent=-2.821, tangentMode=RCTM_User, time=0.5, value=0.625}
          {tangentMode=RCTM_User, time=1, value=0}
         ]}}}
      Scale Curve=&Multiply_Float(A=100, B=$System.FrontendScale)
     ))
    Write to Module Namespace=true
   )
   &FN_Sprite_Attributes(
    Camera Offset:NiagaraFloat=&FN_FrontEndFloatModifier(Gameplay=15, Front End=0)
    Color=&MultiplyLinearColor(
     Linear Color=<rgba 3 2.0562 1.4696 1.5>
     Scale Factor=&MakeVector4FromVectorAndFloat(
      "Vector (RGB)"=&MakeVectorFromLinearColorRGB(Color=$System.TODColorMultiplier)
      "Float (Alpha)"=&FloatFromCurve(FloatCurve=DataInterface() {propertyValues {Curve {keys [
           {arriveTangent=9.0725, interpMode=RCIM_Cubic, leaveTangent=9.0725, tangentMode=RCTM_User, time=0, value=0}
           {arriveTangent=-2.5249, interpMode=RCIM_Cubic, leaveTangent=-2.5249, tangentMode=RCTM_User, time=0.14, value=1}
           {arriveTangent=-0.0405, interpMode=RCIM_Cubic, leaveTangent=-0.0405, tangentMode=RCTM_User, time=1, value=0}
          ]}}})
     )
    )
    Create Camera Offset Attribute=true
    Create Color Attribute=true
    Create SpriteSize Attribute=true
    SpriteSize=&Vector2DFromFloat(Value=&FloatFromCurve(
      FloatCurve=DataInterface() {propertyValues {Curve {keys [{arriveTangent=0.8072, interpMode=RCIM_Cubic, leaveTangent=0.8072, tangentMode=RCTM_User, time=0, value=0.7}, <at 1 1.075>]}}}
      CurveIndex=&Multiply_Float(A=$Particles.NormalizedAge, B=1.1)
      Scale Curve=&Multiply_Float(A=90, B=$System.FrontendScale)
     ))
    Write to Module Namespace=true
   ) {module=FN_Sprite_Attributes001}
   &FN_Sprite_Attributes(
    Camera Offset:NiagaraFloat=&FN_FrontEndFloatModifier(Gameplay=20, Front End=5)
    Color=&MultiplyLinearColor(
     Linear Color=&Multiply_LinearColorByFloat(
      Linear Color=<rgba 1 0.5 0.9803 1>
      Float=&FloatFromCurve(
       FloatCurve=DataInterface() {propertyValues {Curve {keys [
           {interpMode=RCIM_Cubic, tangentMode=RCTM_User, time=0, value=1}
           {arriveTangent=-0.6223, interpMode=RCIM_Cubic, leaveTangent=-0.6223, tangentMode=RCTM_User, time=1, value=0.6}
          ]}}}
       Scale Curve=&FNTimeOfDayScalar(Night Scale=0.9, Morning Scale=1.1, Evening Scale=0.95)
      )
     )
     Scale Factor=&MakeVector4FromVectorAndFloat(
      "Vector (RGB)"=<1 1 1>
      "Float (Alpha)"=&FloatFromCurve(
       FloatCurve=DataInterface() {propertyValues {Curve {keys [
           {arriveTangent=-0.0319, interpMode=RCIM_Cubic, leaveTangent=-0.0319, tangentMode=RCTM_User, time=0, value=1}
           {arriveTangent=-1.1464, interpMode=RCIM_Cubic, leaveTangent=-1.1464, tangentMode=RCTM_User, time=1, value=0}
          ]}}}
       Scale Curve=0.5
      )
     )
    )
    Create Camera Offset Attribute=true
    Create Color Attribute=true
    Create SpriteSize Attribute=true
    SpriteSize=&Vector2DFromFloat(Value=&Multiply_Float(A=100, B=$System.FrontendScale))
    Write to Module Namespace=true
   ) {module=FN_Sprite_Attributes002}
   /SetVariables(Particles.HoloDyn:Vector4f=&MakeVector4(
     X=&FloatFromCurve(FloatCurve=DataInterface() {propertyValues {Curve {keys [
          {arriveTangent=1.5862, interpMode=RCIM_Cubic, leaveTangent=1.5862, tangentMode=RCTM_User, time=0, value=0}
          {arriveTangent=0.159, interpMode=RCIM_Cubic, leaveTangent=0.159, tangentMode=RCTM_User, time=1, value=0.8}
         ]}}})
     Y=&FloatFromCurve(
      FloatCurve=DataInterface() {propertyValues {Curve {keys [
          {arriveTangent=2.5938, interpMode=RCIM_Cubic, leaveTangent=2.5938, tangentMode=RCTM_User, time=0, value=0.2}
          {interpMode=RCIM_Cubic, tangentMode=RCTM_User, time=0.2, value=0.5}
          {tangentMode=RCTM_User, time=1, value=0}
         ]}}}
      Scale Curve=0.7
     )
     Z=&FloatFromCurve(FloatCurve=DataInterface() {propertyValues {Curve {keys [{interpMode=RCIM_Cubic, tangentMode=RCTM_User, time=0, value=1}, {arriveTangent=-0.3137, tangentMode=RCTM_User, time=1, value=0.8}]}}})
     W=&Add_Float(
      A=&RandomRangeFloat(Minimum=0, Maximum=0.25)
      B=&FloatFromCurve(
       FloatCurve=DataInterface() {propertyValues {Curve {keys [{interpMode=RCIM_Cubic, tangentMode=RCTM_User, time=0, value=0}, {interpMode=RCIM_Cubic, tangentMode=RCTM_User, time=1, value=1}]}}}
       Scale Curve=0.25
      )
     )
    )) {isSetParameters=true}
   /SetVariables(Particles.Dark_Glow=&MakeLinearColorFromVectorAndFloat(
     "Vector (RGB)"=&Multiply_Vector(
      A=&MakeVectorFromLinearColorRGB(Color=&MultiplyLinearColor(
        Linear Color=&ColorFromCurve(ColorCurve=DataInterface() {propertyValues {
           AlphaCurve {keys []}
           BlueCurve {keys [<at 0.2119 0.3281 RCIM_Cubic>,<at 0.5555 0.1406>,<at 0.9372 0.05 RCIM_Cubic>]}
           GreenCurve {keys [<at 0.2119 0.3281 RCIM_Cubic>,<at 0.5555 0.1406>,<at 0.9372 0.05 RCIM_Cubic>]}
           RedCurve {keys [<at 0.2119 0.3281 RCIM_Cubic>,<at 0.5555 0.1406>,<at 0.9372 0.05 RCIM_Cubic>]}
          }})
        Scale Factor=&MakeVector4_FromLinearColor(LinearColor=&SelectLinearColorFromArray(
          Array Sampling Mode=::NewEnumerator1
          Direct Array Index=&MakeIntFromFloat(Float=$User.Variant)
          Color Selection Array=DataInterface() {propertyValues {ColorData [
             <rgba 0.7188 0.1093 0.4192 1>
             <rgba 0.0004 0.3596 0.7188 1>
             <rgba 0.1771 0.0985 0.0258 1>
             <rgba 0.0385 0.0637 0.099 0.4783>
             <rgba 0.0399 0.0819 0.099 0.5826>
             <rgba 0.3385 0.1673 0.3317 0.4261>
            ]}}
          Random Seed=?
         ))
       ))
      B=&MakeVectorFromLinearColorRGB(Color=$System.TODColorMultiplier)
     )
     "Float (Alpha)"=&Multiply_Float(
      A=&MakeFloatFromLinearColor(LinearColor=$System.TODColorMultiplier, Channel:ENiagaraFloatFromLinearColorOptions=::NewEnumerator3)
      B=&Multiply_Float(
       A=&MergeHelper_FloatSwitch(0=1, 1=1, 2=1, 3=1, 4=1, 5=0.7, Index=&MakeIntFromFloat(Float=$User.Variant), Num=::NewEnumerator4)
       B=&FloatFromCurve(FloatCurve=DataInterface() {propertyValues {Curve {keys [
            {arriveTangent=-0.3275, interpMode=RCIM_Cubic, leaveTangent=-0.3275, tangentMode=RCTM_User, time=0, value=1}
            {arriveTangent=-0.3009, interpMode=RCIM_Cubic, leaveTangent=-0.3009, tangentMode=RCTM_User, time=1, value=0}
           ]}}})
      )
     )
    )) {isSetParameters=true}
   /SetVariables(Particles.Dark_Glow_Size=&Multiply_Vector2(
     A=&Vector2DFromFloat(Value=100)
     B=&Vector2DFromFloat(Value=&FloatFromCurve(
       FloatCurve=DataInterface() {propertyValues {Curve {keys [
           {arriveTangent=4.1378, interpMode=RCIM_Cubic, leaveTangent=4.1378, tangentMode=RCTM_User, time=0, value=0}
           {interpMode=RCIM_Cubic, leaveTangent=0.5, time=1, value=1.45}
          ]}}}
       Scale Curve=&FN_FrontEndFloatModifier(Gameplay=0.75, Front End=0.4)
      ))
    )) {isSetParameters=true}
   &FN_Sprite_Attributes(
    Camera Offset:NiagaraFloat=&FN_FrontEndFloatModifier(Gameplay=7, Front End=5)
    Color=&Multiply_LinearColorByFloat(
     Linear Color=&MakeLinearColorFromVectorAndFloat("Vector (RGB)"=&MakeVectorFromLinearColorRGB(Color=<rgba 0.209 0.9868 1 0.5>), "Float (Alpha)"=0.6)
     Float=&FloatFromCurve(
      FloatCurve=DataInterface() {propertyValues {Curve {keys [
          {arriveTangent=-0.0405, interpMode=RCIM_Cubic, leaveTangent=-0.0405, tangentMode=RCTM_User, time=0, value=1}
          {arriveTangent=-2.127, interpMode=RCIM_Cubic, leaveTangent=-2.127, tangentMode=RCTM_User, time=1, value=0}
         ]}}}
      CurveIndex=$Particles.Update GalaxySmoke Lifetime
     )
    )
    Create Camera Offset Attribute=true
    Create Color Attribute=true
    Create Sprite Rotation Attribute=true
    Create SpriteSize Attribute=true
    SpriteRotation:NiagaraFloat=&RandomRangeFloat(Minimum=-360, Maximum=360)
    SpriteSize=&Vector2DFromFloat(Value=&FloatFromCurve(
      FloatCurve=DataInterface() {propertyValues {Curve {keys [
          {arriveTangent=1.8336, interpMode=RCIM_Cubic, leaveTangent=1.8336, tangentMode=RCTM_User, time=0, value=0}
          {interpMode=RCIM_Cubic, tangentMode=RCTM_User, time=0.9, value=1}
         ]}}}
      CurveIndex=$Particles.Update GalaxySmoke Lifetime
      Scale Curve=&Multiply_Float(
       A=&RandomRangeFloat(Minimum=100, Maximum=110)
       B=&FN_FrontEndFloatModifier(Gameplay=1, Front End=0.5)
      )
     ))
    Write to Module Namespace=true
   ) {module=FN_Sprite_Attributes003}
   &FN_Mesh_Attributes(
    Color=&MakeLinearColorFromVectorAndFloat(
     "Vector (RGB)"=&Multiply_Vector(
      A=&Lerp_Vector(
       A=&MakeVectorFromLinearColorRGB(Color=<rgba 1 0.1 0.418 1>)
       B=&MakeVectorFromLinearColorRGB(Color=<rgba 0 0.2294 0.974 1>)
       Alpha=&FloatFromCurve(
        FloatCurve=DataInterface() {propertyValues {Curve {keys [{interpMode=RCIM_Cubic, tangentMode=RCTM_User, time=0.4, value=0}, {interpMode=RCIM_Cubic, tangentMode=RCTM_User, time=0.6, value=1}]}}}
        CurveIndex=$Particles.Update GalaxyWave Lifetime
       )
      )
      B=&VectorFromFloat(Value=&FloatFromCurve(
        FloatCurve=DataInterface() {propertyValues {Curve {keys [
            {arriveTangent=11.5613, interpMode=RCIM_Cubic, leaveTangent=11.5613, tangentMode=RCTM_User, time=0, value=0}
            {interpMode=RCIM_Cubic, tangentMode=RCTM_User, time=0.2, value=1}
            {tangentMode=RCTM_User, time=1, value=0}
           ]}}}
        CurveIndex=$Particles.Update GalaxyWave Lifetime
        Scale Curve=3
       ))
     )
     "Float (Alpha)"=&Multiply_Float(
      A=&FloatFromCurve(
       FloatCurve=DataInterface() {propertyValues {Curve {keys [{interpMode=RCIM_Cubic, tangentMode=RCTM_User, time=0, value=1}, {interpMode=RCIM_Cubic, tangentMode=RCTM_User, time=1, value=0}]}}}
       CurveIndex=$Particles.Update GalaxyWave Lifetime
      )
      B=&FNTimeOfDayScalar(Night Scale=0.7, Morning Scale=0.5, Evening Scale=0.5)
     )
    )
    Create Color Attribute=true
    Create Scale Attribute=true
    Scale=&Multiply_Vector(
     A=&VectorFromFloat(Value=0.8)
     B=&VectorFromFloat(Value=&FloatFromCurve(
       FloatCurve=DataInterface() {propertyValues {Curve {keys [
           {arriveTangent=1.152, interpMode=RCIM_Cubic, leaveTangent=1.152, tangentMode=RCTM_User, time=0, value=0.65}
           {arriveTangent=0.0224, interpMode=RCIM_Cubic, leaveTangent=0.0224, tangentMode=RCTM_User, time=1, value=1}
          ]}}}
       CurveIndex=$Particles.Update GalaxyWave Lifetime
       Scale Curve=$System.FrontendScale
      ))
    )
    Write to Module Namespace=true
   )
   /SetVariables(
    Particles.GalaxyCamOffset:NiagaraFloat=&FN_FrontEndFloatModifier(Gameplay=25, Front End=0)
    Particles.GalaxyDynamicPar:Vector4f=&MakeVector4(
     X=&FloatFromCurve(
      FloatCurve=DataInterface() {propertyValues {Curve {keys [{arriveTangent=-2.1015, interpMode=RCIM_Cubic, leaveTangent=-2.1015, tangentMode=RCTM_User, time=0, value=0.5}, <at 1 -0.15>]}}}
      CurveIndex=$Particles.Update GalaxyWave Lifetime
     )
     Y=&FloatFromCurve(
      FloatCurve=DataInterface() {propertyValues {Curve {keys [{arriveTangent=-0.3464, interpMode=RCIM_Cubic, leaveTangent=-0.3464, tangentMode=RCTM_User, time=0, value=1}, <at 1 0.9>]}}}
      CurveIndex=$Particles.Update GalaxyWave Lifetime
     )
     Z=&FloatFromCurve(
      FloatCurve=DataInterface() {propertyValues {Curve {keys [{interpMode=RCIM_Cubic, tangentMode=RCTM_User, time=0, value=1}, {interpMode=RCIM_Cubic, tangentMode=RCTM_User, time=1, value=0.25}]}}}
      CurveIndex=$Particles.Update GalaxyWave Lifetime
     )
     W=0
    )
   ) {isSetParameters=true}
  ]
  rendererClasses [
   &NiagaraSpriteRendererProperties
   &NiagaraLightRendererProperties
   &NiagaraMeshRendererProperties
  ]
  renderers [
   {rendererClass=&NiagaraSpriteRendererProperties}
   {rendererClass=&NiagaraSpriteRendererProperties, rendererIndex=1}
   {rendererClass=&NiagaraSpriteRendererProperties, rendererIndex=2}
   {rendererClass=&NiagaraSpriteRendererProperties, rendererIndex=3}
   {rendererClass=&NiagaraLightRendererProperties, rendererIndex=4}
   {rendererClass=&NiagaraSpriteRendererProperties, rendererIndex=5}
   {rendererClass=&NiagaraSpriteRendererProperties, rendererIndex=6}
   {rendererClass=&NiagaraSpriteRendererProperties, rendererIndex=7}
   {rendererClass=&NiagaraSpriteRendererProperties, rendererIndex=8}
   {rendererClass=&NiagaraSpriteRendererProperties, rendererIndex=9}
   {rendererClass=&NiagaraMeshRendererProperties, rendererIndex=10}
  ]
 }
 Flash_Flare {
  emitterSpawnScript [/SetVariables(Emitter.Flare_Color:LinearColor=&MergeHelper_ColorSwitch(
     0=<rgba 2 0 0.6667 1>
     1=<rgba 0.5413 0.4947 2 1>
     2=<rgba 2 1.1121 0 1>
     3=<rgba 0.7003 1.9465 2 1>
     4=<rgba 1 0.5842 0.9449 1>
     5=<rgba 0.0838 0.2803 1 1>
     Index=&MakeIntFromFloat(Float=$User.Variant)
     Num=::NewEnumerator4
    )) {isSetParameters=true}]
  emitterUpdateScript [
   &EmitterState()
   &SpawnBurst_Instantaneous(Use Spawn Probability=false, UseLoopCountLimit=false)
   &SpawnBurst_Instantaneous(Spawn Count=2, Spawn Group=1, Use Spawn Probability=false, UseLoopCountLimit=false) {module=SpawnBurst_Instantaneous001}
   &SpawnBurst_Instantaneous(
    Spawn Count=&MakeIntFromFloat(Float=&FN_PickFloatDependOnPlatform(
      TargetPlatform=DataInterface() {propertyValues {Platforms {qualityLevelMask=-29}}}
      TargetPlatformActive=&MakeFloatFromInt(Int=12)
      TargetPlatformNoActive=&MakeFloatFromInt(Int=30)
     ))
    Spawn Group=2
    Use Spawn Probability=false
    UseLoopCountLimit=false
   ) {module=SpawnBurst_Instantaneous002}
   &SpawnBurst_Instantaneous(
    Spawn Count=&MakeIntFromFloat(Float=&FN_PickFloatDependOnPlatform(
      TargetPlatform=DataInterface() {propertyValues {Platforms {qualityLevelMask=-29}}}
      TargetPlatformActive=&MakeFloatFromInt(Int=6)
      TargetPlatformNoActive=&MakeFloatFromInt(Int=10)
     ))
    Spawn Group=3
    Use Spawn Probability=false
    UseLoopCountLimit=false
   ) {module=SpawnBurst_Instantaneous003}
  ]
  particleSpawnScript [
   &InitializeParticle(
    Color=&MergeHelper_ColorSwitch(
     0=$Emitter.Flare_Color
     1=&RandomRangeLinearColor(
      Minimum=$Emitter.Flare_Color
      Maximum=&Multiply_LinearColorByFloat(Linear Color=$Emitter.Flare_Color, Float=0.3)
     )
     2=&MergeHelper_ColorSwitch(
      0=&SelectLinearColorFromArray(
       Array Sampling Mode=::NewEnumerator0
       Direct Array Index=?
       Color Selection Array=DataInterface() {propertyValues {ColorData [
          <rgba 2 0.3506 1.03 1>
          <rgba 0.1668 0.7522 2 1>
          <rgba 2 1.9424 0.3033 1>
          <rgba 0.3499 0.2956 2 1>
          <rgba 0.7844 2 0.4817 1>
          <rgba 0.9043 1.8097 2 1>
          <rgba 2 0.4259 0.1102 1>
         ]}}
       Random Seed=0
      )
      1=&SelectLinearColorFromArray(
       Array Sampling Mode=::NewEnumerator0
       Direct Array Index=?
       Color Selection Array=DataInterface() {propertyValues {ColorData [<rgba 0.4389 0.3113 2 1>, <rgba 0.2736 0.3385 2 1>, <rgba 0.3984 0.4237 1 1>]}}
       Random Seed=0
      )
      2=&SelectLinearColorFromArray(
       Array Sampling Mode=::NewEnumerator0
       Direct Array Index=?
       Color Selection Array=DataInterface() {propertyValues {ColorData [<rgba 0 0 0 3>, <rgba 1.5 0.8991 0.2581 1>, <rgba 2 1.4758 0.0304 1>]}}
       Random Seed=0
      )
      3=&SelectLinearColorFromArray(
       Array Sampling Mode=::NewEnumerator0
       Direct Array Index=?
       Color Selection Array=DataInterface() {propertyValues {ColorData [<rgba 0.7246 1.4757 2 3>, <rgba 0.587 1.5 1.4798 1>]}}
       Random Seed=0
      )
      4=&SelectLinearColorFromArray(
       Array Sampling Mode=::NewEnumerator0
       Direct Array Index=?
       Color Selection Array=DataInterface() {propertyValues {ColorData [<rgba 0.4 1.3848 2 1>, <rgba 1.5892 0.8 2 1>]}}
       Random Seed=0
      )
      5=&SelectLinearColorFromArray(
       Array Sampling Mode=::NewEnumerator0
       Direct Array Index=?
       Color Selection Array=DataInterface() {propertyValues {ColorData [<rgba 0 0.5641 2 1>, <rgba 1.3396 0 2 1>]}}
       Random Seed=0
      )
      Index=&MakeIntFromFloat(Float=$User.Variant)
      Num=::NewEnumerator4
     )
     3=&MergeHelper_ColorSwitch(
      0=<rgba 0 0 0 0>
      1=<rgba 0 0 0 0>
      2=&RandomRangeLinearColor(Minimum=<rgba 2 1.41 0.23 1>, Maximum=<rgba 2 1 0 1>)
      3=<rgba 0.7646 0.9519 1 1>
      4=&ColorFromCurve(
       ColorCurve=DataInterface() {propertyValues {
         AlphaCurve {keys [<at 0 1 RCIM_Cubic>]}
         BlueCurve {keys [<at 0.2316 1 RCIM_Cubic>,<at 0.3038 1 RCIM_Cubic>]}
         GreenCurve {keys [<at 0.2316 0.6924 RCIM_Cubic>,<at 0.3038 0.4 RCIM_Cubic>]}
         RedCurve {keys [<at 0.2316 0.2 RCIM_Cubic>,<at 0.3038 0.7946 RCIM_Cubic>]}
        }}
       CurveIndex=&RandomRangeFloat(Minimum=0, Maximum=1)
      )
      5=&LerpLinearColors(
       StartColor=<rgba 1.3396 0 2 1>
       EndColor=<rgba 0 0.5641 2 1>
       LerpFactor=&UniformAOrBFloat(A=0, Evaluation Type=::NewEnumerator0, Recalculate Random Each Loop=true, Override Seed=false)
      )
      Index=&MakeIntFromFloat(Float=$User.Variant)
      Num=::NewEnumerator4
     )
     Index=$Emitter.SpawnGroup
     Num=::NewEnumerator2
    )
    Lifetime=&MergeHelper_FloatSwitch(
     0=0.18
     1=&RandomRangeFloat(Minimum=0.1, Maximum=0.15)
     2=&RandomRangeFloat(Minimum=0.05, Maximum=0.45)
     3=&RandomRangeFloat(Minimum=0.3, Maximum=0.6)
     Index=$Emitter.SpawnGroup
     Num=::NewEnumerator2
    )
    Mesh Scale Mode=::NewEnumerator3
    Mesh Uniform Scale=&Multiply_Float(A=&RandomRangeFloat(Minimum=0.08, Maximum=0.11), B=$System.FrontendScale)
    Sprite Rotation Angle=&MergeHelper_FloatSwitch(
     0=0
     1=0
     2=0
     3=&MergeHelper_FloatSwitch(
      0=0
      1=0
      2=&RandomRangeFloat(Minimum=0, Maximum=1)
      3=&RandomRangeFloat(Minimum=0, Maximum=1)
      4=&RandomRangeFloat(Minimum=0, Maximum=1)
      5=&RandomRangeFloat(Minimum=0, Maximum=1)
      Index=&MakeIntFromFloat(Float=$User.Variant)
      Num=::NewEnumerator4
     )
     Index=$Emitter.SpawnGroup
     Num=::NewEnumerator2
    )
    Sprite Rotation Mode=::NewEnumerator1
    Sprite Size=&MergeHelper_Vector2Switch(
     0=&FN_FrontEndVector2Modifier(Front End:Vector2f=<40 150>, Gameplay:Vector2f=<80 240>)
     1=&Multiply_Vector2D(
      A=&UniformRangedVector2D(Minimum=<20 40>, Maximum=<25 70>)
      B=&FN_FrontEndVector2Modifier(Front End:Vector2f=<0.5 0.7>, Gameplay:Vector2f=<1 1>)
     )
     2=&Multiply_Vector2D_ByFloat(
      Float=&Multiply_Float(
       A=&FN_FrontEndFloatModifier(Gameplay=1, Front End=0.45)
       B=&RandomRangeFloat(Minimum=4.5, Maximum=5)
      )
      Vector2D=<1 1.5>
     )
     3=&Vector2DFromFloat(Value=&Multiply_Float(
       A=$System.FrontendScale
       B=&Multiply_Float(
        A=&RandomRangeFloat(Minimum=10, Maximum=14)
        B=&MergeHelper_FloatSwitch(
         0=0
         1=0
         2=1.35
         3=1
         4=0.75
         5=1.5
         Index=&MakeIntFromFloat(Float=$User.Variant)
         Num=::NewEnumerator4
        )
       )
      ))
     4=?
     Index=$Emitter.SpawnGroup
     Num=::NewEnumerator2
    )
    Sprite Size Mode=::NewEnumerator4
    Sprite UV Mode=::NewEnumerator1
    UsePositionOffset=false
    Write Material Random=true
   )
   &ShapeLocation(
    Cylinder Height=&MergeHelper_FloatSwitch(0=0, 1=0, 2=80, 3=0, Index=$Emitter.SpawnGroup, Num=::NewEnumerator2)
    Cylinder Height Midpoint=0
    Cylinder Radius=&MergeHelper_FloatSwitch(0=0, 1=0, 2=3, 3=3, Index=$Emitter.SpawnGroup, Num=::NewEnumerator2)
    MaskSpawnGroup=false
    Override Seed=false
    Rotation Axis=$System.DirectionVector
    Rotation Mode=::NewEnumerator1
    Shape Origin:Vector3f=$System.FEImpactOffset
    Shape Primitive=::NewEnumerator1
    Surface Only=false
    UseOwnerScale=true
   )
   &CameraOffset(Camera Offset Amount=&MergeHelper_FloatSwitch(0=30, 1=25, 2=20, 3=20, Index=$Emitter.SpawnGroup, Num=::NewEnumerator2))
   &SubUVAnimation(
    End Frame Range Override=3
    Evaluation Type=::NewEnumerator0
    "Sprite Renderer "=DataInterface() {propertyValues {SpriteRenderer=&self:Flash_Flare_1.NiagaraSpriteRendererProperties_0}}
    SubUV Animation Mode:ENiagara_SubUVLookupModeV2=::NewEnumerator2
    UseEndFrame=true
    UseStartFrame=true
   )
   &DynamicMaterialParameters(
    Index 0 Param 1=0.2
    Index 0 Param 2=1
    Index 3 Param 1=1.5
    Index 3 Param 2=1
    Module.Index0Float0Write=true
    Module.Index0Float1Write=true
    Module.Index0Float2Write=true
    Module.Index0Float3Write=true
    Module.Index3Float0Write=true
    Module.Index3Float1Write=true
    Module.Index3Float2Write=true
    Module.Index3Float3Write=true
    Write Parameter Index 3=::NewEnumerator1
   )
   /SetVariables(Particles.DynamicParameterAlt:Vector4f=&MakeVector4(X=1, Y=1, Z=1, W=1)) {isSetParameters=true}
   &AddVelocity(
    Module.Use Falloff Distance=false
    Override Seed=false
    Velocity Mode=::NewEnumerator1
    Velocity Speed=&Multiply_Float(
     A=&MergeHelper_FloatSwitch(
      0=0
      1=20
      2=&RandomRangeFloat(Minimum=225, Maximum=300)
      3=&RandomRangeFloat(Minimum=225, Maximum=300)
      Index=$Emitter.SpawnGroup
      Num=::NewEnumerator2
     )
     B=$System.FrontEndScale
    )
   ) {module=AddVelocity001}
   &AddVelocity(
    Rotation Coordinate Space=::Simulation
    Velocity=$System.DirectionVector
    Velocity Speed Scale=&MergeHelper_FloatSwitch(
     0=&Multiply_Float(A=&RandomRangeFloat(Minimum=200, Maximum=300), B=$System.FrontEndScale)
     1=&Multiply_Float(A=100, B=$System.FrontEndScale)
     2=&Multiply_Float(A=&RandomRangeFloat(Minimum=0, Maximum=750), B=$System.FrontEndScale)
     3=&Multiply_Float(A=&RandomRangeFloat(Minimum=0, Maximum=750), B=$System.FrontEndScale)
     Index=$Emitter.SpawnGroup
     Num=::NewEnumerator2
    )
   )
   /SetVariables() {isSetParameters=true}
   /SetVariables(Particles.VisibilityTag=$Emitter.SpawnGroup) {isSetParameters=true}
  ]
  particleUpdateScript [
   &ParticleState()
   &ScaleMeshSize(Scale Factor=&VectorFromFloat(Value=&FloatFromCurve(FloatCurve=DataInterface() {propertyValues {Curve {keys [<at 0 1 RCIM_Cubic>,<at 1 0 RCIM_Cubic>]}}})))
   &SubUVAnimation(
    Output Alternate Renderer Binding=true
    Play Rate:NiagaraFloat=&RandomRangeFloat(Minimum=0.5, Maximum=1)
    Playback Scale Curve=DataInterface() {propertyValues {Curve {keys [
        {interpMode=RCIM_Cubic, tangentMode=RCTM_User, time=0.25, value=1}
        {arriveTangent=-1.2656, interpMode=RCIM_Cubic, leaveTangent=-1.2656, tangentMode=RCTM_User, time=1, value=0}
       ]}}}
    Random Start Frame=true
    "Sprite Renderer "=DataInterface() {propertyValues {SpriteRenderer=&self:Flash_Flare_1.NiagaraSpriteRendererProperties_7}}
    SubUV Animation Mode:ENiagara_SubUVLookupModeV2=::NewEnumerator3
    UseEndFrame=false
    UsePlayRateScale=false
    UseStartFrame=false
   ) {module=SubUVAnimation001}
   &AccelerationForce(Acceleration=&MergeHelper_VectorSwitch(0=<0 0 0>, 1=<0 0 0>, 2=<0 0 -500>, 3=<0 0 -500>, Index=$Particles.VisibilityTag, Num=::NewEnumerator2))
   &Drag(
    Drag=&MergeHelper_FloatSwitch(
     0=0
     1=0
     2=&RandomRangeFloat(Minimum=5, Maximum=8)
     3=&RandomRangeFloat(Minimum=5, Maximum=8)
     Index=$Particles.VisibilityTag
     Num=::NewEnumerator2
    )
    Rotational Drag:NiagaraFloat=&MergeHelper_FloatSwitch(0=0, 1=0, 2=1, 3=1, Index=$Particles.VisibilityTag, Num=::NewEnumerator2)
    Use Linear Drag=true
    Use Rotational Drag=true
   )
   &SolveForcesAndVelocity(Clamp Velocity=false, Limit Acceleration=false)
   &ScaleSpriteSizeBySpeed(
    Max Scale Factor=&MergeHelper_Vector2Switch(0=<1 1>, 1=<1 1>, 2=<0.2 5>, 3=<1 1>, 4=?, Index=$Particles.VisibilityTag, Num=::NewEnumerator2)
    Min Scale Factor=&MergeHelper_Vector2Switch(0=<1 1>, 1=<1 1>, 2=<0.2 0.2>, 3=<1 1>, 4=?, Index=$Particles.VisibilityTag, Num=::NewEnumerator2)
   )
   &ScaleSpriteSize(
    Scale Factor=&MergeHelper_Vector2Switch(
     0=&Vector2DFromCurve(
      Vector2Curve=DataInterface() {propertyValues {
        XCurve {keys [{tangentMode=RCTM_User, time=0, value=0}, <at 0.2 1.075>, {tangentMode=RCTM_User, time=1, value=-0.1}]}
        YCurve {keys [{tangentMode=RCTM_User, time=0, value=0}, <at 0.1 1.025>, {tangentMode=RCTM_User, time=1, value=0.8}]}
       }}
      ScaleCurve=<1 1>
     )
     1=&Vector2DFromCurve(
      Vector2Curve=DataInterface() {propertyValues {
        XCurve {keys [<at 0 1 RCIM_Cubic>,<at 1 0 RCIM_Cubic>]}
        YCurve {keys [{arriveTangent=3.9291, interpMode=RCIM_Cubic, leaveTangent=3.9291, tangentMode=RCTM_User, time=0, value=0}, <at 1 2 RCIM_Cubic>]}
       }}
      ScaleCurve=&Multiply_Vector2D_ByFloat(Float=1, Vector2D=<1.5 1>)
     )
     2=&Vector2DFromFloat(Value=&FloatFromCurve(FloatCurve=DataInterface() {propertyValues {Curve {keys [{arriveTangent=-6.725, interpMode=RCIM_Cubic, leaveTangent=-6.725, tangentMode=RCTM_User, time=0.65, value=1}, <at 1 0 RCIM_Cubic>]}}}))
     3=&Vector2DFromFloat(Value=&FloatFromCurve(FloatCurve=DataInterface() {propertyValues {Curve {keys [{arriveTangent=-6.725, interpMode=RCIM_Cubic, leaveTangent=-6.725, tangentMode=RCTM_User, time=0.65, value=1}, <at 1 0 RCIM_Cubic>]}}}))
     4=?
     Index=$Particles.VisibilityTag
     Num=::NewEnumerator2
    )
    Scale Sprite Size Mode=::NewEnumerator2
   )
   &ScaleColor(
    Scale Alpha=&MergeHelper_FloatSwitch(
     0=1
     1=&FloatFromCurve(
      FloatCurve=DataInterface() {propertyValues {Curve {keys [<at 0 1 RCIM_Cubic>,<at 1 0>]}}}
      Scale Curve=&FNTimeOfDayScalar(Night Scale=0.5, Morning Scale=1, Evening Scale=0.5)
     )
     2=1
     3=1
     Index=$Particles.VisibilityTag
     Num=::NewEnumerator2
    )
    Scale RGB=&Multiply_Vector(
     A=&MakeVectorFromLinearColorRGB(Color=$System.TODColorMultiplier)
     B=&MergeHelper_VectorSwitch(
      0=<1 1 1>
      1=&VectorFromFloat(Value=&FloatFromCurve(FloatCurve=DataInterface() {propertyValues {Curve {keys [{arriveTangent=-15.8795, interpMode=RCIM_Cubic, leaveTangent=-15.8795, tangentMode=RCTM_User, time=0, value=10}, <at 1 1 RCIM_Cubic>]}}}))
      2=&VectorFromFloat(Value=&FloatFromCurve(
        FloatCurve=DataInterface() {propertyValues {Curve {keys [
            {interpMode=RCIM_Cubic, tangentMode=RCTM_User, time=0, value=1}
            <at 0.3 0.2>
            {interpMode=RCIM_Cubic, tangentMode=RCTM_User, time=1, value=0}
           ]}}}
        Scale Curve=10
       ))
      3=&VectorFromFloat(Value=&FloatFromCurve(
        FloatCurve=DataInterface() {propertyValues {Curve {keys [
            {interpMode=RCIM_Cubic, tangentMode=RCTM_User, time=0, value=1}
            <at 0.3 0.2>
            {interpMode=RCIM_Cubic, tangentMode=RCTM_User, time=1, value=0}
           ]}}}
        Scale Curve=10
       ))
      Index=$Particles.VisibilityTag
      Num=::NewEnumerator2
     )
    )
    ScaleA=true
    ScaleRGB=true
   )
   /SetVariables(
    Particles.GlowColor=&MakeLinearColorFromVectorAndFloat(
     "Vector (RGB)"=&MakeVectorFromLinearColorRGB(Color=$Particles.Initial.Color)
     "Float (Alpha)"=&FloatFromCurve(FloatCurve=DataInterface() {propertyValues {Curve {keys [<at 0.5 1>,<at 1 0>]}}}, Scale Curve=0.04)
    )
    Particles.GlowSize=&Multiply_Vector2D(A=$Particles.SpriteSize, B=<8 2>)
   ) {isSetParameters=true}
  ]
  rendererClasses [&NiagaraSpriteRendererProperties, &NiagaraMeshRendererProperties]
  renderers [
   {rendererClass=&NiagaraSpriteRendererProperties}
   {rendererClass=&NiagaraSpriteRendererProperties, rendererIndex=1}
   {rendererClass=&NiagaraSpriteRendererProperties, rendererIndex=2}
   {rendererClass=&NiagaraSpriteRendererProperties, rendererIndex=3}
   {rendererClass=&NiagaraSpriteRendererProperties, rendererIndex=4}
   {rendererClass=&NiagaraSpriteRendererProperties, rendererIndex=5}
   {rendererClass=&NiagaraSpriteRendererProperties, rendererIndex=6}
   {rendererClass=&NiagaraSpriteRendererProperties, rendererIndex=7}
   {rendererClass=&NiagaraMeshRendererProperties, rendererIndex=8}
  ]
 }
 ImpactSmoke {
  emitterUpdateScript [
   &EmitterState()
   &SpawnBurst_Instantaneous(
    Spawn Count=&RandomRangeInt(Minimum=2, Maximum=3)
    Use Spawn Probability=false
    UseLoopCountLimit=false
   )
  ]
  particleSpawnScript [
   &InitializeParticle(
    Color Maximum=<rgba 0.76 1 1 0.2>
    Color Minimum=<rgba 0.8 0.917 1 0.15>
    Color Mode=::NewEnumerator2
    Lifetime Max=0.4
    Lifetime Min=0.3
    Lifetime Mode=::NewEnumerator1
    Sprite Size Mode=::NewEnumerator3
    Uniform Sprite Size=&Multiply_Float(
     A=&FN_FrontEndFloatModifier(Gameplay=1, Front End=0.45)
     B=&RandomRangeFloat(Minimum=110, Maximum=140)
    )
    UsePositionOffset=false
    Write Material Random=true
   )
   &CameraOffset(Camera Offset Amount=&RandomRangeFloat(Minimum=5, Maximum=10, Evaluation Type=::NewEnumerator1))
   &ShapeLocation(
    Hemisphere Distribution=<1 0.5>
    MaskSpawnGroup=false
    Override Seed=false
    Rotation Angle=90
    Rotation Mode=::NewEnumerator1
    Shape Origin:Vector3f=$System.FEImpactOffset
    Sphere Radius=&MakeCustomFloatFromBool(Bool=$User.bIsFrontEnd, True Float=5, False Float=10)
    UseOwnerScale=true
   )
   &AddVelocity(Velocity=&Multiply_VectorByFloat(Vector=$Particles.ShapeLocation.ShapeVector, Float=10))
  ]
  particleUpdateScript [
   &ParticleState()
   &ScaleSpriteSize(Uniform Curve Sprite Scale=DataInterface() {propertyValues {Curve {keys [
        {arriveTangent=1.7761, interpMode=RCIM_Cubic, leaveTangent=1.7761, tangentMode=RCTM_User, time=0, value=0}
        {interpMode=RCIM_Cubic, tangentMode=RCTM_User, time=1, value=1}
       ]}}})
   &ScaleColor(
    Scale Alpha=&FloatFromCurve(FloatCurve=DataInterface() {propertyValues {Curve {keys [
         {arriveTangent=2.4358, interpMode=RCIM_Cubic, leaveTangent=2.4358, tangentMode=RCTM_User, time=0, value=0}
         <at 0.25 0.5 RCIM_Cubic>
         <at 1 0 RCIM_Cubic>
        ]}}})
    Scale RGB=&MakeVectorFromLinearColorRGB(Color=&MergeHelper_ColorSwitch(
      0=&ColorFromCurve(ColorCurve=DataInterface() {propertyValues {
         AlphaCurve {keys [<at 0 1 RCIM_Cubic>,<at 1 0 RCIM_Cubic>]}
         BlueCurve {keys [<at 0.1955 0.882 RCIM_Cubic>,<at 0.7637 1 RCIM_Cubic>]}
         GreenCurve {keys [<at 0.1955 0.629 RCIM_Cubic>,<at 0.7637 1 RCIM_Cubic>]}
         RedCurve {keys [<at 0.1955 1 RCIM_Cubic>,<at 0.7637 1 RCIM_Cubic>]}
        }})
      1=&ColorFromCurve(ColorCurve=DataInterface() {propertyValues {
         AlphaCurve {keys [<at 0 1 RCIM_Cubic>,<at 1 0 RCIM_Cubic>]}
         BlueCurve {keys [<at 0.1955 0.5669 RCIM_Cubic>,<at 0.5354 1 RCIM_Cubic>]}
         GreenCurve {keys [<at 0.1955 3.0248 RCIM_Cubic>,<at 0.5354 1 RCIM_Cubic>]}
         RedCurve {keys [<at 0.1955 5 RCIM_Cubic>,<at 0.5354 1 RCIM_Cubic>]}
        }})
      2=&ColorFromCurve(ColorCurve=DataInterface() {propertyValues {
         AlphaCurve {keys [<at 0 1 RCIM_Cubic>,<at 1 0 RCIM_Cubic>]}
         BlueCurve {keys [<at 0.1955 0.3499 RCIM_Cubic>,<at 0.5354 1 RCIM_Cubic>]}
         GreenCurve {keys [<at 0.1955 1.9539 RCIM_Cubic>,<at 0.5354 1 RCIM_Cubic>]}
         RedCurve {keys [<at 0.1955 5 RCIM_Cubic>,<at 0.5354 1 RCIM_Cubic>]}
        }})
      3=&ColorFromCurve(ColorCurve=DataInterface() {propertyValues {
         AlphaCurve {keys [<at 0 1 RCIM_Cubic>,<at 1 0 RCIM_Cubic>]}
         BlueCurve {keys [<at 0.1955 5 RCIM_Cubic>,<at 0.5354 1 RCIM_Cubic>]}
         GreenCurve {keys [<at 0.1955 1.3142 RCIM_Cubic>,<at 0.5354 1 RCIM_Cubic>]}
         RedCurve {keys [<at 0.1955 4.0763 RCIM_Cubic>,<at 0.5354 1 RCIM_Cubic>]}
        }})
      4=&ColorFromCurve(ColorCurve=DataInterface() {propertyValues {
         AlphaCurve {keys [<at 0 1 RCIM_Cubic>,<at 1 0 RCIM_Cubic>]}
         BlueCurve {keys [<at 0.1955 5 RCIM_Cubic>,<at 0.5354 1 RCIM_Cubic>]}
         GreenCurve {keys [<at 0.1955 3.6217 RCIM_Cubic>,<at 0.5354 1 RCIM_Cubic>]}
         RedCurve {keys [<at 0.1955 1.4745 RCIM_Cubic>,<at 0.5354 1 RCIM_Cubic>]}
        }})
      5=&ColorFromCurve(ColorCurve=DataInterface() {propertyValues {
         AlphaCurve {keys [<at 0 1 RCIM_Cubic>,<at 1 0 RCIM_Cubic>]}
         BlueCurve {keys [<at 0.1955 5 RCIM_Cubic>,<at 0.5354 1 RCIM_Cubic>]}
         GreenCurve {keys [<at 0.1955 2.1403 RCIM_Cubic>,<at 0.5354 1 RCIM_Cubic>]}
         RedCurve {keys [<at 0.1955 2.7942 RCIM_Cubic>,<at 0.5354 1 RCIM_Cubic>]}
        }})
      Index=&MakeIntFromFloat(Float=$User.Variant)
      Num=::NewEnumerator4
     ))
    ScaleA=true
    ScaleRGB=true
   )
   &DynamicMaterialParameters(
    Index 0 Param 1=&Add_Float(
     A=&RandomRangeFloat(Minimum=0, Maximum=1)
     B=&FloatFromCurve(
      FloatCurve=DataInterface() {propertyValues {Curve {keys [
          <at 0 0 RCIM_Cubic>
          {arriveTangent=1.5947, interpMode=RCIM_Cubic, leaveTangent=1.5947, tangentMode=RCTM_User, time=0.2557, value=0.5257}
          <at 1 1 RCIM_Cubic>
         ]}}}
      CurveIndex=$Particles.Age
      Scale Curve=0.25
     )
    )
    Module.Index0Float0Write=true
    Module.Index0Float1Write=true
    Module.Index0Float2Write=true
    Module.Index0Float3Write=true
   ) {module=DynamicMaterialParameters001}
   &SolveForcesAndVelocity(Clamp Velocity=false, Limit Acceleration=false)
  ]
  rendererClasses [&NiagaraSpriteRendererProperties]
  renderers [
   {rendererClass=&NiagaraSpriteRendererProperties}
   {rendererClass=&NiagaraSpriteRendererProperties, rendererIndex=1}
  ]
 }
 Sparks {
  emitterUpdateScript [
   &EmitterState()
   &SpawnBurst_Instantaneous(
    Spawn Count=&MakeCustomIntFromBool(Bool=$User.bIsFrontEndPreview, True INT=30, False INT=&RandomRangeInt(Minimum=15, Maximum=20))
    Use Spawn Probability=false
    UseLoopCountLimit=false
   )
  ]
  particleSpawnScript [
   &InitializeParticle(
    Color=&MergeHelper_ColorSwitch(
     0=&SelectLinearColorFromArray(
      Array Sampling Mode=::NewEnumerator0
      Direct Array Index=?
      Color Selection Array=DataInterface() {propertyValues {ColorData [
         <rgba 2 0.3506 1.03 1>
         <rgba 0.1668 0.7522 2 1>
         <rgba 2 1.9424 0.3033 1>
         <rgba 0.3499 0.2956 2 1>
         <rgba 0.7844 2 0.4817 1>
         <rgba 0.9043 1.8097 2 1>
         <rgba 2 0.4259 0.1102 1>
        ]}}
      Random Seed=0
     )
     1=&SelectLinearColorFromArray(
      Array Sampling Mode=::NewEnumerator0
      Direct Array Index=?
      Color Selection Array=DataInterface() {propertyValues {ColorData [<rgba 0.4389 0.3113 2 1>, <rgba 0.2736 0.3385 2 1>, <rgba 0.3984 0.4237 1 1>]}}
      Random Seed=0
     )
     2=&SelectLinearColorFromArray(
      Array Sampling Mode=::NewEnumerator0
      Direct Array Index=?
      Color Selection Array=DataInterface() {propertyValues {ColorData [<rgba 0 0 0 3>, <rgba 1.5 0.8991 0.2581 1>, <rgba 2 1.4758 0.0304 1>]}}
      Random Seed=0
     )
     3=&SelectLinearColorFromArray(
      Array Sampling Mode=::NewEnumerator0
      Direct Array Index=?
      Color Selection Array=DataInterface() {propertyValues {ColorData [<rgba 0.7246 1.4757 2 3>, <rgba 0.587 1.5 1.4798 1>]}}
      Random Seed=0
     )
     4=&SelectLinearColorFromArray(
      Array Sampling Mode=::NewEnumerator0
      Direct Array Index=?
      Color Selection Array=DataInterface() {propertyValues {ColorData [<rgba 0.4 1.3848 2 1>, <rgba 1.5892 0.8 2 1>]}}
      Random Seed=0
     )
     5=&SelectLinearColorFromArray(
      Array Sampling Mode=::NewEnumerator0
      Direct Array Index=?
      Color Selection Array=DataInterface() {propertyValues {ColorData [<rgba 0 0.5641 2 1>, <rgba 1.3396 0 2 1>]}}
      Random Seed=0
     )
     Index=&MakeIntFromFloat(Float=$User.Variant)
     Num=::NewEnumerator4
    )
    Lifetime Max=0.45
    Lifetime Min=0.05
    Lifetime Mode=::NewEnumerator1
    Sprite Size Mode=::NewEnumerator3
    Uniform Sprite Size=&Multiply_Float(A=$System.FrontEndScale, B=&RandomRangeFloat(Minimum=3, Maximum=5))
    UsePositionOffset=false
    Write Material Random=true
   )
   &ShapeLocation(
    MaskSpawnGroup=false
    Override Seed=false
    Shape Origin:Vector3f=$System.FEImpactOffset
    Sphere Radius=&Multiply_Float(A=2, B=&FN_FrontEndFloatModifier(Gameplay=0, Front End=0.6))
    UseOwnerScale=true
   )
   &CameraOffset(Camera Offset Amount=25)
   &AddVelocity(
    Cone Angle=180
    Cone Axis=&FN_FrontEndVector3Modifier(Gameplay=<1 0 0>, Front End=<-1 0 0>)
    Module.UseConeFalloff=false
    Module.UseVelDistribution=false
    Override Seed=false
    Velocity Mode=::NewEnumerator2
    Velocity Speed=&Multiply_Float(A=&RandomRangeFloat(Minimum=400, Maximum=750), B=$System.FrontEndScale)
   )
  ]
  particleUpdateScript [
   &ParticleState()
   &AccelerationForce(Acceleration=&RandomRangeVector(Minimum=<0 0 -500>, Maximum=<0 0 -500>))
   &Drag(Drag=&RandomRangeFloat(Minimum=5, Maximum=8), Use Linear Drag=true, Use Rotational Drag=true)
   &SolveForcesAndVelocity(Clamp Velocity=false, Limit Acceleration=false)
   &ScaleSpriteSizeBySpeed(Max Scale Factor=<0.5 10>, Min Scale Factor=<0.5 1>)
   &ScaleSpriteSize(
    Uniform Curve Scale=&Multiply_Float(A=1, B=&FN_FrontEndFloatModifier(Gameplay=1, Front End=0.5))
    Uniform Curve Sprite Scale=DataInterface() {propertyValues {Curve {keys [{arriveTangent=-6.725, interpMode=RCIM_Cubic, leaveTangent=-6.725, tangentMode=RCTM_User, time=0.65, value=1}, <at 1 0 RCIM_Cubic>]}}}
   )
   &ScaleColor(
    Scale RGB=&Multiply_Vector(
     A=&MakeVectorFromLinearColorRGB(Color=$System.TODColorMultiplier)
     B=&VectorFromFloat(Value=&FloatFromCurve(
       FloatCurve=DataInterface() {propertyValues {Curve {keys [
           {interpMode=RCIM_Cubic, tangentMode=RCTM_User, time=0, value=1}
           <at 0.3 0.2>
           {interpMode=RCIM_Cubic, tangentMode=RCTM_User, time=1, value=0}
          ]}}}
       Scale Curve=10
      ))
    )
    ScaleA=true
    ScaleRGB=true
   )
   /SetVariables(
    Particles.GlowColor=&MakeLinearColorFromVectorAndFloat(
     "Vector (RGB)"=&MakeVectorFromLinearColorRGB(Color=$Particles.Color)
     "Float (Alpha)"=&FloatFromCurve(FloatCurve=DataInterface() {propertyValues {Curve {keys [<at 0.5 1>,<at 1 0>]}}}, Scale Curve=0.04)
    )
    Particles.GlowSize=&Multiply_Vector2D(A=$Particles.SpriteSize, B=<8 2>)
   ) {isSetParameters=true}
  ]
  rendererClasses [&NiagaraSpriteRendererProperties]
  renderers [
   {rendererClass=&NiagaraSpriteRendererProperties}
   {rendererClass=&NiagaraSpriteRendererProperties, rendererIndex=1}
  ]
 }
}
summary {
 systemName=NS_Pickaxe_TimeTrotter_Impact
 emitters [
  {
   emitterName=Flash_Elements
   rendererClasses [
    &NiagaraSpriteRendererProperties
    &NiagaraLightRendererProperties
    &NiagaraMeshRendererProperties
   ]
  }
  {emitterName=ImpactSmoke, rendererClasses [&NiagaraSpriteRendererProperties]}
  {emitterName=Flash_Flare, rendererClasses [&NiagaraSpriteRendererProperties, &NiagaraMeshRendererProperties]}
  {emitterName=Sparks, rendererClasses [&NiagaraSpriteRendererProperties]}
  {
   emitterName=Directional_FN__PhysMatDebris001
   rendererClasses [&NiagaraSpriteRendererProperties, &NiagaraMeshRendererProperties]
  }
 ]
 userVariables [
  User.bIsFrontEnd {type=NiagaraBool}
  User.bIsFrontEndPreview {type=NiagaraBool}
  User.PhysicalSurface {type=NiagaraInt32}
  User.SwingSectionIndex {type=NiagaraInt32}
  User.Variant {type=NiagaraFloat, defaultValue=0, description=""}
 ]
}
