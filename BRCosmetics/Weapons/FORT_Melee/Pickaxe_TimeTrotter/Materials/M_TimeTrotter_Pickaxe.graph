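# Master material for the TimeTrotter pickaxe, layering animated gradient emissive and a clock display onto the 2023 character parent PreFX/PostFX functions.
# A tech-art map masks two panning gradients: one projected triplanar-style from pre-skinned local position, another using UV-space polar wheel and strip coordinates.
# The backpack's shared clock function, confined by a UV sphere mask, is max-combined into emissive.
# That mask also feeds temporal responsiveness.
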Material /BRCosmetics/Weapons/FORT_Melee/Pickaxe_TimeTrotter/Materials/M_TimeTrotter_Pickaxe
graph {
 properties {
  bAutomaticallySetUsageInEditor=false
  bUseMaterialAttributes=true
  bUsedWithMeshDeformer=true
  bUsedWithMorphTargets=true
  bUsedWithSkeletalMesh=true
  pixelDepthOffsetMode=PDOM_Legacy
  shadingModel=MSM_FromMaterialExpression
  usedShadingModels=Default Lit | Subsurface | Clear Coat | Subsurface Profile | Cloth
 }
 nodes [
  ComponentMask() {b=true}
  ComponentMask(None=Constant4Vector() {constant=<rgba 0.848 0.1612 0 0>}.RGBA) {g=true, r=true}
  n1 = ScalarParameter() {defaultValue=1, group=06_Advanced Emissive, parameterName=EmissivePannerScale}
  n2 = Divide(
   A=VertexInterpolator(VS=LocalPosition() {localOrigin=InstancePreSkinning}.XYZ).PS
   B=ScalarParameter() {defaultValue=40, group=06_Advanced Emissive, parameterName=EmissiveTextureConstantScaleDivisor, sortPriority=4}
  )
  n3 = VectorParameter() {group=06_Advanced Emissive, parameterName=Emissive Panning Speed, sortPriority=2, defaultValue=<rgba 0 0.2 0.2 1>}
  n4 = Frac(None=Multiply(A=Time(), B=@n3.RGB))
  n5 = Add(A=Multiply(A=@n1, B=@n2), B=@n4)
  n6 = Add(A=Multiply(A=@n1, B=@n2), B=@n4)
  n7 = VertexInterpolator(VS=PreSkinnedNormal())
  n8 = ComponentMask(None=@n7.PS) {g=true}
  n9 = ScalarParameter() {defaultValue=1, group=06_Advanced Emissive, parameterName=TriPlannerAxisBlendStrength, sortPriority=4}
  n10 = ComponentMask(None=MaterialFunctionCall(In=Abs(None=@n8), Contrast=@n9) {materialFunction=&CheapContrast}.Result) {g=true}
  n11 = Multiply(A=@n1, B=@n2)
  n12 = MaterialFunctionCall(In=Abs(None=ComponentMask(None=@n7.PS) {b=true}), Contrast=@n9) {materialFunction=&CheapContrast}
  n13 = LinearInterpolate(
   A=LinearInterpolate(A=ComponentMask(None=@n5) {b=true, g=true}, B=ComponentMask(None=@n6) {b=true, r=true}, Alpha=Saturate(None=@n10))
   B=ComponentMask(None=Add(A=@n11, B=@n4)) {g=true, r=true}
   Alpha=Saturate(None=ComponentMask(None=@n12.Result) {b=true})
  )
  ComponentMask(None=@n13) {g=true, r=true}
  n14 = MaterialFunctionCall() {materialFunction=&MF_CharacterParent_2023_PreFX}
  n15 = GetMaterialAttributes(None=@n14.Result) {attributeGetTypes [#1]}
  n16 = TextureSampleParameter2D!(
   UVs=LinearInterpolate(
    A=ComponentMask(None=@n13) {g=true, r=true}
    B=MaterialFunctionCall(XY=@n13) {materialFunction=&Swizzle}.YX
    Alpha=ScalarParameter() {group=TimeTrotter, parameterName=SwitchDirection}
   )
   Level
   Apply View MipBias
  ) {
   group=TimeTrotter
   mipValueMode=TMVM_MipLevel
   parameterName=PanningEmissiveColorGradient
   samplerSource=SSM_Wrap_WorldGroupSettings
   sortPriority=6
   texture=/Game/Global/Textures/Default/Masks_and_Gradients/3pxRGBImage
  }
  n17 = TextureSampleParameter2D(UVs=TextureCoordinate()) {
   group=TimeTrotter
   mipValueMode=TMVM_MipLevel
   parameterName=TechArtMap
   samplerSource=SSM_Wrap_WorldGroupSettings
   samplerType=SAMPLERTYPE_Masks
   texture=/BRCosmetics/Weapons/FORT_Melee/Pickaxe_TimeTrotter/Textures/T_TimeTrotter_Pickaxe_TechArt
  }
  n18 = Subtract(A=TextureCoordinate(), B=ComponentMask(None=Constant4Vector() {constant=<rgba 0.848 0.1612 0 0>}.RGBA) {g=true, r=true}) {constB=0.5}
  n19 = Divide(A=Arctangent2Fast(Y=ComponentMask(None=@n18) {r=true}, X=ComponentMask(None=@n18) {g=true})) {constB=6.2831}
  n20 = TextureCoordinate()
  n21 = Step(Y=ComponentMask(None=@n20) {r=true}) {constX=0.3}
  n22 = Multiply(
   A=Time()
   B=LinearInterpolate(
    A=ScalarParameter() {defaultValue=0.1, group=TimeTrotter, parameterName=EmissiveGradient.WheelSpeed}
    B=ScalarParameter() {defaultValue=0.1, group=TimeTrotter, parameterName=EmissiveGradient.PanningSpeed}
    Alpha=@n21
   )
  )
  n23 = Add(A=Multiply(A=Add(A=@n19) {constB=0.5}, B=ScalarParameter() {defaultValue=1, group=TimeTrotter, parameterName=Loop_MainWheel}), B=@n22)
  n24 = ComponentMask(None=@n20) {g=true}
  n25 = ScalarParameter() {defaultValue=0.5, group=TimeTrotter, parameterName=loops_FrontStrip}
  n26 = Add(A=Multiply(A=@n21, B=Multiply(A=@n24, B=@n25) {constB=0.5}), B=@n22)
  n27 = TextureCoordinate()
  n28 = TextureSampleParameter2D!(UVs=Frac(None=Add(A=@n23, B=@n26)), "DDX(UVs)"=DDX(Value=@n27), "DDY(UVs)"=DDY(Value=@n27), Apply View MipBias) {
   group=TimeTrotter
   mipValueMode=TMVM_Derivative
   parameterName=PanningEmissiveColorGradient2
   samplerSource=SSM_Wrap_WorldGroupSettings
   sortPriority=6
   texture=&3pxRGBImage
  }
  n29 = LinearInterpolate(
   A=Multiply(A=@n15.Emissive Color, B=NamedRerouteUsage())
   B=Add(A=Multiply(A=@n16.RGB, B=@n17.B), B=Multiply(A=@n28.RGB, B=@n17.G))
   Alpha=Add(A=@n17.B, B=@n17.B)
  )
  n30 = SphereMask(
   A=TextureCoordinate()
   B=ComponentMask(None=VectorParameter() {group=TimeTrotter_Clock, parameterName=UVPosition, defaultValue=<rgba 0.848 0.1612 0 1>}.RGB) {g=true, r=true}
  ) {attenuationRadius=0.095}
  n31 = Subtract(
   A=Multiply(
    A=MaterialFunctionCall() {materialFunction=/BRCosmetics/Accessories/FORT_Backpacks/Backpack_F_MED_TimeTrotter/Materials/MF_TimeTrotter_Clock}.Clock
    B=@n30
   )
   B=@n17.R
  )
  MaterialFunctionCall(
   Material=SetMaterialAttributes(MaterialAttributes=@n15.MaterialAttributes, Emissive Color=Max(A=@n29, B=@n31)) {attributeSetTypes [#1]}
   M=@n14.M
   Pre Skinned Local Position=@n14.M
   Distance Blend=@n14.M
   Fresnel=@n14.M
  ) {materialFunction=&MF_CharacterParent_2023_PostFX}
  NamedRerouteDeclaration(None=Subtract(A=@n17.R, B=@n17.R)) {name=EmissiveMaskOut, nodeColor=<rgba 0 0.9062 1 1>}
  Step(Y=@n24) {constX=0.855}
  TemporalResponsivenessOutput!(None=Max(A=ComponentMask(None=@n14.Tech Art Mask) {b=true}, B=@n30))
 ]
}
