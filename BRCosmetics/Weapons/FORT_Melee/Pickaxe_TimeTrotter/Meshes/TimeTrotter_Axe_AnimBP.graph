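# Animation Blueprint for the TimeTrotter pickaxe that adjusts its two dynamic side-piece bones on top of the input pose.
# Four chained additive bone-space Transform Bone nodes follow the input: a fixed X offset for each piece, then a zero-default translation for each.
# Each update computes a game-time sine wave on X into the "Pieces move" vector, though no link from it to the bone nodes is shown.
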AnimBlueprint /BRCosmetics/Weapons/FORT_Melee/Pickaxe_TimeTrotter/Meshes/TimeTrotter_Axe_AnimBP
target_skeleton=&/TimeTrotter_Axe_Skeleton
graphs [
 AnimGraph {nodes [
   AnimGraphNode_SubInput_0 = Input Pose()
   n1 = "Transform (Modify) Bone - Bone: dyn_side_piece_01"(ComponentPose=Local To Component(LocalPose=@AnimGraphNode_SubInput_0), Alpha=1, Translation="1.5,0,0", Rotation="0,0,0", Scale="1,1,1") {translationMode=BMM_Additive, translationSpace=BCS_BoneSpace, boneToModify {boneName=dyn_side_piece_01}, translation {x=1.5}}
   n2 = "Transform (Modify) Bone - Bone: dyn_side_piece_02"(ComponentPose=@n1, Alpha=1, Translation="1.5,0,0", Rotation="0,0,0", Scale="1,1,1") {translationMode=BMM_Additive, translationSpace=BCS_BoneSpace, boneToModify {boneName=dyn_side_piece_02}, translation {x=1.5}}
   n3 = "Transform (Modify) Bone - Bone: dyn_side_piece_01"(ComponentPose=@n2, Alpha=1, Translation="0,0,0", Rotation="0,0,0", Scale="1,1,1") {
    translationMode=BMM_Additive
    translationSpace=BCS_BoneSpace
    alphaScaleBiasClamp {interpSpeedDecreasing=2, interpSpeedIncreasing=2}
    boneToModify {boneName=dyn_side_piece_01}
   }
   n4 = "Transform (Modify) Bone - Bone: dyn_side_piece_02"(ComponentPose=@n3, Alpha=1, Translation="0,0,0", Rotation="0,0,0", Scale="1,1,1") {
    translationMode=BMM_Additive
    translationSpace=BCS_BoneSpace
    alphaScaleBiasClamp {interpSpeedDecreasing=2, interpSpeedIncreasing=2}
    boneToModify {boneName=dyn_side_piece_02}
   }
   Output Pose(Result=Component To Local(ComponentPose=@n4)) {layerGroup=None}
  ]}
 EventGraph {nodes [
   Event BlueprintUpdateAnimation()
   n1 = "Sin (Radians)"(A=float * float(A=GetGameTimeInSeconds(), B=4.5) {is=title K2Node_PromotableOperator})
   > Set Pieces move(Pieces move:Vector=MakeVector(X=float * float(A=@n1, B=1.5), Y=0)) -> (Output_Get:Vector)
   TryGetPawnOwner()
  ]}
]
variables [Pieces move {type=Vector}]
